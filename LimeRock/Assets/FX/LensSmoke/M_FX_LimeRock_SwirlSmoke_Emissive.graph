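# Unlit translucent sprite material for LimeRock swirling smoke particles, supporting Niagara and particle sprites.
# An 8x8 SubUV smoke texture is tinted by particle color, with green driving emissive and red driving opacity.
# Switches add near-camera, depth and close-up distance fades, plus MF_SmoothLod distance-based opacity and WPO scaling.
# Depth fade is bypassed on low quality and mobile feature levels, while forward shading also skips near fade.
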Material /LimeRock/Assets/FX/LensSmoke/M_FX_LimeRock_SwirlSmoke_Emissive
graph {
 properties {
  allowTranslucentCustomDepthWrites=true
  bUsedWithNiagaraSprites=true
  bUsedWithParticleSprites=true
  blendMode=BLEND_Translucent
  opacityMaskClipValue=0.01
  pixelDepthOffsetMode=PDOM_Legacy
  shadingModel=MSM_Unlit
  translucencyLightingMode=TLM_VolumetricPerVertexDirectional
  usedShadingModels=Unlit
 }
 nodes [
  n1 = LinearInterpolate(Alpha=MaterialFunctionCall!(
    Blend Range=ScalarParameter() {defaultValue=200, parameterName=CloseFade}
    Start Offset=ScalarParameter() {defaultValue=-150, parameterName=CloseFade_Offset}
   ) {materialFunction=/Engine/Functions/Engine_MaterialFunctions03/Particles/Distance_Blend}.Result)
  LinearInterpolate(
   A=@n1
   Alpha=MaterialFunctionCall(Blend Range=Constant() {r=5000}, Start Offset=Constant() {r=-1050}) {materialFunction=&Distance_Blend}.Result
  ) {constB=0}
  n2 = ParticleSubUV!(UVs, Tex, Apply View MipBias) {samplerType=SAMPLERTYPE_LinearColor, texture=/Game/Effects/Fort_Effects/Textures/Smoke/T_Smoke_8x8_01}
  n3 = ParticleColor()
  n4 = Multiply(A=@n3.RGB, B=@n2.G)
  MaterialFunctionCall(Emissive=@n4) {materialFunction=&MF_FX_EmissiveAdjustments}
  n5 = MaterialFunctionCall!(
   NearFadeLength=ScalarParameter() {defaultValue=1500, group=ScaleSize, parameterName="NearFadeLength (S)", sortPriority=1}
   NearFadeOffset=ScalarParameter() {group=ScaleSize, parameterName=NearFadeOffset, sortPriority=1}
   NearFadeCurve=ScalarParameter() {group=ScaleSize, parameterName=NearFadeCurve, sortPriority=1}
   NearFadeStartingScale=ScalarParameter() {defaultValue=1, group=ScaleSize, parameterName="NearFadeStartingScale (S)"}
   NearFadeEndingScale
   FarFadeLength=ScalarParameter() {defaultValue=2500, group=ScaleSize, parameterName="FarFadeLength (S)", sortPriority=4}
   FarFadeOffset
   FarFadeCurve
   FarFadeEndingScale=ScalarParameter() {defaultValue=4, group=ScaleSize, parameterName=FarFadeEndingScale_HIGH, sortPriority=3}
   Opacity Near Distance Bias=ScalarParameter() {group=ScaleSize, parameterName=OpacityNear_Bias, sortPriority=5}
   Opacity Far Distance Bias=ScalarParameter() {group=ScaleSize, parameterName=OpacityFar_Bias, sortPriority=6}
  ) {materialFunction=/Game/Effects/Fort_Effects/Materials/_Functions/MF_SmoothLod}
  n6 = Constant() {r=1}
  n7 = Multiply(A=@n3.A, B=@n2.R)
  n8 = ScalarParameter() {defaultValue=250, parameterName="Distance (S)"}
  n9 = Multiply(
   A=@n7
   B=StaticSwitchParameter(
    True=MaterialFunctionCall(Distance=@n8) {materialFunction=&Near_Camera_Fade}.Result
    False=Constant() {r=1}
   ) {parameterName=UseNearFade}
  )
  n10 = StaticSwitchParameter(True=DepthFade(Opacity=@n9, FadeDistance=ScalarParameter() {defaultValue=250, parameterName=FadeDistance}), False=@n9) {parameterName=UseDepthFade}
  n11 = FeatureLevelSwitch(Default=@n10, ES2=@n9, ES3_1=@n9, SM4_REMOVED=@n10)
  n12 = QualitySwitch(Default=@n11, Low=@n9, High=@n11, Medium=@n11)
  Multiply(
   A=StaticSwitchParameter(True=@n1, False=StaticSwitchParameter(True=@n5.Opacity, False=@n6) {parameterName=Scale_Opacity_When_Close?}) {parameterName=UseLowendCloseUpFade?}
   B=ShadingPathSwitch(Default=@n12, Deferred=@n12, Forward=@n7)
  )
  Reroute(None=@n4)
  StaticSwitchParameter(True=@n5.WPO, False=@n6) {parameterName=Scale_Down_When_Close?}
 ]
}
