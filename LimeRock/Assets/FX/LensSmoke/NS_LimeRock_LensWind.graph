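# Niagara system for a LimeRock lens wind effect: a gust of dark smoke, small green debris and speed lines, all pulled toward the camera by camera offsets.
# The three enabled sprite emitters spawn in short timed windows, offset along System.WindDirection from the simulation position, and fly back against that vector.
# Speed lines swirl with curl noise and stretch by velocity.
# Other smoke, debris, speed-line and post-process emitters are disabled.
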NiagaraSystem /LimeRock/Assets/FX/LensSmoke/NS_LimeRock_LensWind
dependencies {
 usedDataInterfaces [
  &NiagaraDataInterfaceCurve
  &NiagaraDataInterfaceGBuffer
  &NiagaraDataInterfaceColorCurve
  &NiagaraDataInterfaceSpriteRendererInfo
  &NiagaraDataInterfaceVector2DCurve
 ]
 usedDynamicInputs [&NormalizeVector, &SubtractVectorFromPosition]
 usedModules [
  &SystemState
  &self:SystemSpawnScript.SystemScriptSource.SystemScriptGraph.NiagaraNodeAssignment_0.SetVariables_#1
  &self:SystemSpawnScript.SystemScriptSource.SystemScriptGraph.NiagaraNodeAssignment_1.SetVariables_#2
 ]
}
emitters {
 Debris001 {
  bEnabled=false
  simTarget=GPUComputeSim
  emitterUpdateScript [
   &EmitterState()
   &SpawnRate(SpawnRate=40, Use Spawn Probability=false)
  ]
  particleSpawnScript [
   &InitializeParticle(
    Color=&RandomRangeLinearColor(Minimum=<rgba 0.3372 0.2195 0.0369 1>, Maximum=<rgba 0.8958 0.4472 0.07 1>)
    Lifetime=2
    Mesh Scale Mode=::NewEnumerator3
    Sprite Rotation Mode=::NewEnumerator2
    Sprite Size Mode=::NewEnumerator3
    Uniform Sprite Size=&RandomRangeFloat(Minimum=10, Maximum=20)
    UsePositionOffset=false
    Write Material Random=true
   )
   /Niagara/Modules/Math/PlaceParticlesOnDepthBuffer_GPU(Max Z Depth=8000, Module.Use Max Z Depth=true) {enabled=false}
   /Game/Effects/Niagara/Modules/Spawn/FN_Spawn_GBuffer_Color() {enabled=false}
   &KillParticles(Kill Particles=&SetBoolByFloatComparison(
     A=&VectorLength(Vector=&Subtract_Vector(A=$Particles.InitialPos, B=Linked() {linkedVariable {name=Particles.Position, type=NiagaraPosition}}))
     B=5000
    )) {enabled=false}
   &KillParticles(Kill Particles=&Boolean_NotOperation(Bool=$Output.PlaceParticlesOnDepthBuffer_GPU.ValidSample)) {enabled=false, module=KillParticles001}
   &AddVelocity(
    Coordinate Space=::World
    Velocity=&RotateVectorByQuaternion(
     Vector To Rotate=<2000 0 50>
     Quaternion=&FindQuatBetweenTwoVectors(From Vector=<-1 0 0>, To Vector=$System.WindDirection)
    )
   )
   &CameraOffset(Camera Offset Amount=250)
   &ShapeLocation(
    MaskSpawnGroup=false
    Non Uniform Scale=<1 1 0.1>
    Offset Coordinate Space=::World
    Offset Mode=::NewEnumerator0
    Override Seed=false
    Sphere Radius=2500
    UseOwnerScale=true
   )
  ]
  particleUpdateScript [
   &ParticleState()
   &ScaleSpriteSize(Uniform Curve Sprite Scale=DataInterface() {propertyValues {Curve {keys [
        {interpMode=RCIM_Cubic, tangentMode=RCTM_User, time=0, value=0}
        {interpMode=RCIM_Cubic, tangentMode=RCTM_User, time=0.5, value=1}
        {interpMode=RCIM_Cubic, tangentMode=RCTM_User, time=0.8, value=0.95}
        {tangentMode=RCTM_User, time=1, value=0}
       ]}}})
   &SpriteRotationRate(Rotation Rate=&RandomRangeFloat(Minimum=-300, Maximum=300))
   &CurlNoiseForce(Noise Frequency=1000, Noise Strength=500, Pan Noise=false)
   &SolveForcesAndVelocity(Clamp Velocity=true, Limit Acceleration=true, Speed Limit=850)
   &ScaleColor(
    Scale Alpha=&FloatFromCurve(
     FloatCurve=DataInterface() {propertyValues {Curve {keys [
         {interpMode=RCIM_Cubic, tangentMode=RCTM_User, time=0, value=0}
         {interpMode=RCIM_Cubic, tangentMode=RCTM_User, time=0.1, value=1}
         <at 0.8 0.95>
         {tangentMode=RCTM_User, time=1, value=0}
        ]}}}
     Scale Curve=1.5
    )
    ScaleA=true
    ScaleRGB=true
   )
   &DynamicMaterialParameters(
    Index 0 Param 1=&FloatFromCurve(FloatCurve=DataInterface() {propertyValues {Curve {keys [<at 0 0>,<at 1 0>]}}})
    Index 0 Param 2=2
    Index 0 Param 3=&RandomRangeFloat(Minimum=500, Maximum=500)
    Index 0 Param 4=0.25
    Index 1 Param 1=512
    Module.Param0WriteEnabled=true
    Module.Param0WriteEnabled1=true
    Module.Param1WriteEnabled=true
    Module.Param1WriteEnabled1=true
    Module.Param2WriteEnabled=true
    Module.Param2WriteEnabled1=true
    Module.Param3WriteEnabled=true
    Module.Param3WriteEnabled1=true
    Write Parameter Index 1=true
   ) {module=DynamicMaterialParameters001}
   &SubUVAnimation(
    Playback Scale Curve=DataInterface() {propertyValues {Curve {keys [{interpMode=RCIM_Cubic, tangentMode=RCTM_User, time=0, value=0}, {interpMode=RCIM_Cubic, tangentMode=RCTM_User, time=1, value=1}]}}}
    SubUV Animation Mode:ENiagara_SubUVLookupModeV2=::NewEnumerator3
    UseEndFrame=true
    UsePlayRateScale=true
    UseStartFrame=true
   ) {module=SubUVAnimation002}
  ]
  rendererClasses [&NiagaraSpriteRendererProperties]
  renderers [{rendererClass=&NiagaraSpriteRendererProperties}]
 }
 PostProcess {
  bEnabled=false
  emitterUpdateScript [
   &EmitterState(
    Inactive Response=::NewEnumerator1
    Life Cycle Mode=::NewEnumerator1
    Loop Behavior=::NewEnumerator1
    Loop Delay=1
    Loop Duration=6
    UseLoopDelay=true
   )
   &SpawnBurst_Instantaneous(Use Spawn Probability=true, UseLoopCountLimit=false)
  ]
  particleSpawnScript [
   /Niagara/Modules/Spawn/Initialization/InitializeParticle(
    Lifetime=0.5
    Mesh Scale=<5 5 5>
    Module.Write Color=true
    Module.Write Lifetime=true
    Module.Write Mass=true
    Module.Write Material Random=true
    Module.Write Position=false
    Module.Write Scale=false
    Module.Write SpriteRotation=false
    Module.Write SpriteSize=false
   ) {enabled=false}
   /Niagara/Modules/Spawn/Initialization/V2/InitializeParticle(Lifetime=8, Mass Mode=::NewEnumerator1, Sprite Size Mode=::NewEnumerator4, UsePositionOffset=false, Write Material Random=true) {module=InitializeParticle001}
  ]
  particleUpdateScript [
   &ParticleState()
   /Game/Effects/Niagara/Modules/ComponentRenderer/FN_PostProcess(Blend:NiagaraFloat=&FloatFromCurve(
     FloatCurve=DataInterface() {propertyValues {Curve {keys [
         {arriveTangent=0.2844, interpMode=RCIM_Cubic, leaveTangent=0.2844, tangentMode=RCTM_User, time=0, value=0}
         {arriveTangent=0.0294, interpMode=RCIM_Cubic, leaveTangent=0.0294, tangentMode=RCTM_User, time=2, value=1}
         {interpMode=RCIM_Cubic, tangentMode=RCTM_User, time=5, value=1}
         <at 6 0 RCIM_Cubic>
        ]}}}
     CurveIndex=$Particles.Age
    ))
  ]
  rendererClasses [&NiagaraComponentRendererProperties]
  renderers [{rendererClass=&NiagaraComponentRendererProperties, rendererIndex=0}]
 }
 Small_Debris {
  emitterUpdateScript [
   &EmitterState(Life Cycle Mode=::NewEnumerator1, Loop Behavior=::NewEnumerator1, Loop Duration=1.7, UseLoopDelay=false)
   &SpawnRate(SpawnRate=120, Use Spawn Probability=false)
  ]
  particleSpawnScript [
   &InitializeParticle(
    Color=&RandomRangeLinearColor(Minimum=<rgba 0.0163 0.1667 0.0079 1>, Maximum=<rgba 0.2184 0.8958 0.028 1>)
    Lifetime=4
    Mesh Scale Mode=::NewEnumerator3
    Sprite Rotation Mode=::NewEnumerator2
    Sprite Size Mode=::NewEnumerator3
    Uniform Sprite Size=&RandomRangeFloat(Minimum=10, Maximum=20)
    UsePositionOffset=false
    Write Material Random=true
   )
   &CameraOffset(Camera Offset Amount=250)
   &ShapeLocation(
    MaskSpawnGroup=false
    Offset=&Multiply_VectorByFloat(Vector=$System.WindDirection, Float=10000)
    Offset Coordinate Space=::World
    Offset Mode=::NewEnumerator0
    Override Seed=false
    Rotation Coordinate Space=::World
    Shape Origin=/Niagara/DynamicInputs/Helpers/SimulationPosition()
    Sphere Radius=800
    UseOwnerScale=false
   ) {module=ShapeLocation001}
   &AddVelocity(Rotation Coordinate Space=::World, Velocity=$System.WindDirection, Velocity Speed Scale=-4000)
  ]
  particleUpdateScript [
   &ParticleState()
   &ScaleSpriteSize(Uniform Curve Sprite Scale=DataInterface() {propertyValues {Curve {keys [
        {interpMode=RCIM_Cubic, tangentMode=RCTM_User, time=0, value=0}
        {interpMode=RCIM_Cubic, tangentMode=RCTM_User, time=0.5, value=1}
        {interpMode=RCIM_Cubic, tangentMode=RCTM_User, time=0.8, value=0.95}
        {tangentMode=RCTM_User, time=1, value=0}
       ]}}})
   &SpriteRotationRate(Rotation Rate=&RandomRangeFloat(Minimum=-300, Maximum=300))
   &SolveForcesAndVelocity(Clamp Velocity=false, Limit Acceleration=false) {module=SolveForcesAndVelocity001}
   &ScaleColor(
    Scale Alpha=&FloatFromCurve(
     FloatCurve=DataInterface() {propertyValues {Curve {keys [
         {interpMode=RCIM_Cubic, tangentMode=RCTM_User, time=0, value=0}
         {interpMode=RCIM_Cubic, tangentMode=RCTM_User, time=0.1, value=1}
         <at 0.8 0.95>
         {tangentMode=RCTM_User, time=1, value=0}
        ]}}}
     Scale Curve=1.5
    )
    ScaleA=true
    ScaleRGB=true
   )
   &DynamicMaterialParameters(
    Index 0 Param 1=&FloatFromCurve(FloatCurve=DataInterface() {propertyValues {Curve {keys [<at 0 0>,<at 1 0>]}}})
    Index 0 Param 2=2
    Index 0 Param 3=&RandomRangeFloat(Minimum=500, Maximum=500)
    Index 0 Param 4=0.25
    Index 1 Param 1=512
    Module.Param0WriteEnabled=true
    Module.Param0WriteEnabled1=true
    Module.Param1WriteEnabled=true
    Module.Param1WriteEnabled1=true
    Module.Param2WriteEnabled=true
    Module.Param2WriteEnabled1=true
    Module.Param3WriteEnabled=true
    Module.Param3WriteEnabled1=true
    Write Parameter Index 1=true
   ) {module=DynamicMaterialParameters001}
   &SubUVAnimation(
    Playback Scale Curve=DataInterface() {propertyValues {Curve {keys [{interpMode=RCIM_Cubic, tangentMode=RCTM_User, time=0, value=0}, {interpMode=RCIM_Cubic, tangentMode=RCTM_User, time=1, value=1}]}}}
    "Sprite Renderer "=DataInterface() {propertyValues {SpriteRenderer=&self:Small_Debris_0.NiagaraSpriteRendererProperties_0}}
    SubUV Animation Mode:ENiagara_SubUVLookupModeV2=::NewEnumerator3
    UseEndFrame=true
    UsePlayRateScale=true
    UseStartFrame=true
   ) {module=SubUVAnimation002}
  ]
  rendererClasses [&NiagaraSpriteRendererProperties]
  renderers [{rendererClass=&NiagaraSpriteRendererProperties}]
 }
 Smoke {
  bEnabled=false
  emitterUpdateScript [
   &EmitterState()
   &SpawnRate(SpawnRate=10, Use Spawn Probability=false)
  ]
  particleSpawnScript [
   &InitializeParticle(
    Color=<rgba 1 1 1 0.2>
    Lifetime=2
    Mesh Scale Mode=::NewEnumerator3
    Sprite Rotation Mode=::NewEnumerator2
    Sprite Size Mode=::NewEnumerator3
    Uniform Sprite Size=1500
    UsePositionOffset=false
    Write Material Random=true
   )
   &PlaceParticlesOnDepthBuffer_GPU(Max Z Depth=8000, Module.Use Max Z Depth=true) {enabled=false}
   &FN_Spawn_GBuffer_Color() {enabled=false}
   &KillParticles(Kill Particles=&SetBoolByFloatComparison(
     A=&VectorLength(Vector=&Subtract_Vector(A=$Particles.InitialPos, B=Linked() {linkedVariable {name=Particles.Position, type=NiagaraPosition}}))
     B=5000
    )) {enabled=false}
   &KillParticles(Kill Particles=&Boolean_NotOperation(Bool=$Output.PlaceParticlesOnDepthBuffer_GPU.ValidSample)) {enabled=false, module=KillParticles001}
   &AddVelocity(
    Coordinate Space=::World
    Velocity=&RotateVectorByQuaternion(
     Vector To Rotate=<2000 0 50>
     Quaternion=&FindQuatBetweenTwoVectors(From Vector=<-1 0 0>, To Vector=$System.WindDirection)
    )
   )
   &CameraOffset(Camera Offset Amount=250)
   &ShapeLocation(
    MaskSpawnGroup=false
    Non Uniform Scale=<1 1 0>
    Offset Coordinate Space=::World
    Offset Mode=::NewEnumerator0
    Override Seed=false
    Sphere Radius=2500
    UseOwnerScale=true
   )
  ]
  particleUpdateScript [
   &ParticleState()
   &SolveForcesAndVelocity(Clamp Velocity=true, Limit Acceleration=true, Speed Limit=850)
   &ScaleColor(
    Scale Alpha=&FloatFromCurve(
     FloatCurve=DataInterface() {propertyValues {Curve {keys [
         {interpMode=RCIM_Cubic, tangentMode=RCTM_User, time=0, value=0}
         {interpMode=RCIM_Cubic, tangentMode=RCTM_User, time=0.5, value=1}
         <at 0.8 0.95>
         {tangentMode=RCTM_User, time=1, value=0}
        ]}}}
     Scale Curve=1.5
    )
    Scale RGB=&MakeVectorFromLinearColorRGB(Color=<rgba 0 0 0 1>)
    ScaleA=true
    ScaleRGB=false
   )
   &SubUVAnimation(
    Start Frame:NiagaraInt32=&RandomRangeInt(Minimum=12, Maximum=20)
    SubUV Animation Mode:ENiagara_SubUVLookupMode=::NewEnumerator1
   )
   &DynamicMaterialParameters(
    Index 0 Param 1=250
    Module.Param0WriteEnabled=true
    Module.Param1WriteEnabled=true
    Module.Param2WriteEnabled=true
    Module.Param3WriteEnabled=true
   )
   /SetVariables(Particles.NewColor:LinearColor=&MultiplyLinearColor(
     Linear Color=&MultiplyLinearColor(
      Linear Color=$Particles.Color
      Scale Factor=&MakeVector4_FromLinearColor(LinearColor=&ColorFromCurve(
        ColorCurve=DataInterface() {propertyValues {
          AlphaCurve {keys [<at 0 1 RCIM_Cubic>,<at 1 1 RCIM_Cubic>]}
          BlueCurve {keys [
            {arriveTangent=1.3581, interpMode=RCIM_Cubic, time=0, value=0.15}
            <at 0.0736 0.25>
            <at 0.1163 0.2448>
            <at 0.15 0.0328>
            <at 0.2091 0.25>
            <at 0.25 0.25>
            <at 0.455 0.5>
            <at 0.6597 0.5>
            <at 0.7304 0.25>
            <at 0.8114 0.25>
            <at 0.8468 0.0741>
            <at 0.8983 0.0649>
            <at 0.9248 0.15>
            <at 1 0.15 RCIM_Cubic>
           ]}
          GreenCurve {keys [
            {arriveTangent=1.3581, interpMode=RCIM_Cubic, time=0, value=0.15}
            <at 0.0736 0.25>
            <at 0.1163 0.2448>
            <at 0.15 0.0477>
            <at 0.2091 0.25>
            <at 0.25 0.25>
            <at 0.455 0.5>
            <at 0.6597 0.5>
            <at 0.7304 0.25>
            <at 0.8114 0.25>
            <at 0.8468 0.0741>
            <at 0.8983 0.0649>
            <at 0.9248 0.15>
            <at 1 0.15 RCIM_Cubic>
           ]}
          RedCurve {keys [
            {arriveTangent=1.3581, interpMode=RCIM_Cubic, time=0, value=0.15}
            <at 0.0736 0.25>
            <at 0.1163 0.2448>
            <at 0.15 0.1615>
            <at 0.2091 0.25>
            <at 0.25 0.25>
            <at 0.455 0.5>
            <at 0.6597 0.5>
            <at 0.7304 0.25>
            <at 0.8114 0.25>
            <at 0.8468 0.1094>
            <at 0.8983 0.1094>
            <at 0.9248 0.15>
            <at 1 0.15 RCIM_Cubic>
           ]}
         }}
        CurveIndex=Linked() {linkedVariable {name=NPC.FortniteNPC.TimeOfDay_0-1, type=NiagaraFloat}}
       ))
     )
     Scale Factor=<1 1 1 1>
    )) {isSetParameters=true, module=SetVariables_#3}
  ]
  rendererClasses [&NiagaraSpriteRendererProperties]
  renderers [{rendererClass=&NiagaraSpriteRendererProperties}]
 }
 Smoke001 {
  bEnabled=false
  simTarget=GPUComputeSim
  emitterUpdateScript [
   &EmitterState()
   &SpawnRate(SpawnRate=10, Use Spawn Probability=false)
  ]
  particleSpawnScript [
   &InitializeParticle(
    Color=<rgba 1 1 1 0.2>
    Lifetime=2
    Mesh Scale Mode=::NewEnumerator3
    Sprite Rotation Mode=::NewEnumerator2
    Sprite Size Mode=::NewEnumerator3
    Uniform Sprite Size=1500
    UsePositionOffset=false
    Write Material Random=true
   )
   &PlaceParticlesOnDepthBuffer_GPU(Max Z Depth=8000, Module.Use Max Z Depth=true) {enabled=false}
   &FN_Spawn_GBuffer_Color() {enabled=false}
   &KillParticles(Kill Particles=&SetBoolByFloatComparison(
     A=&VectorLength(Vector=&Subtract_Vector(A=$Particles.InitialPos, B=Linked() {linkedVariable {name=Particles.Position, type=NiagaraPosition}}))
     B=5000
    )) {enabled=false}
   &KillParticles(Kill Particles=&Boolean_NotOperation(Bool=$Output.PlaceParticlesOnDepthBuffer_GPU.ValidSample)) {enabled=false, module=KillParticles001}
   &AddVelocity(
    Coordinate Space=::World
    Velocity=&RotateVectorByQuaternion(
     Vector To Rotate=<2000 0 50>
     Quaternion=&FindQuatBetweenTwoVectors(From Vector=<-1 0 0>, To Vector=$System.WindDirection)
    )
   )
   &CameraOffset(Camera Offset Amount=250)
   &ShapeLocation(
    MaskSpawnGroup=false
    Non Uniform Scale=<1 1 0>
    Offset Coordinate Space=::World
    Offset Mode=::NewEnumerator0
    Override Seed=false
    Sphere Radius=2500
    UseOwnerScale=true
   )
  ]
  particleUpdateScript [
   &ParticleState()
   &SolveForcesAndVelocity(Clamp Velocity=true, Limit Acceleration=true, Speed Limit=850)
   &ScaleColor(
    Scale Alpha=&FloatFromCurve(
     FloatCurve=DataInterface() {propertyValues {Curve {keys [
         {interpMode=RCIM_Cubic, tangentMode=RCTM_User, time=0, value=0}
         {interpMode=RCIM_Cubic, tangentMode=RCTM_User, time=0.5, value=1}
         <at 0.8 0.95>
         {tangentMode=RCTM_User, time=1, value=0}
        ]}}}
     Scale Curve=1.5
    )
    Scale RGB=&MakeVectorFromLinearColorRGB(Color=<rgba 0 0 0 1>)
    ScaleA=true
    ScaleRGB=false
   )
   &SubUVAnimation(
    Start Frame:NiagaraInt32=&RandomRangeInt(Minimum=12, Maximum=20)
    SubUV Animation Mode:ENiagara_SubUVLookupMode=::NewEnumerator1
   )
   &DynamicMaterialParameters(
    Index 0 Param 1=250
    Module.Param0WriteEnabled=true
    Module.Param1WriteEnabled=true
    Module.Param2WriteEnabled=true
    Module.Param3WriteEnabled=true
   )
   /SetVariables(Particles.NewColor:LinearColor=&MultiplyLinearColor(
     Linear Color=&MultiplyLinearColor(
      Linear Color=$Particles.Color
      Scale Factor=&MakeVector4_FromLinearColor(LinearColor=&ColorFromCurve(
        ColorCurve=DataInterface() {propertyValues {
          AlphaCurve {keys [<at 0 1 RCIM_Cubic>,<at 1 1 RCIM_Cubic>]}
          BlueCurve {keys [
            {arriveTangent=1.3581, interpMode=RCIM_Cubic, time=0, value=0.15}
            <at 0.0736 0.25>
            <at 0.1163 0.2448>
            <at 0.15 0.0328>
            <at 0.2091 0.25>
            <at 0.25 0.25>
            <at 0.455 0.5>
            <at 0.6597 0.5>
            <at 0.7304 0.25>
            <at 0.8114 0.25>
            <at 0.8468 0.0741>
            <at 0.8983 0.0649>
            <at 0.9248 0.15>
            <at 1 0.15 RCIM_Cubic>
           ]}
          GreenCurve {keys [
            {arriveTangent=1.3581, interpMode=RCIM_Cubic, time=0, value=0.15}
            <at 0.0736 0.25>
            <at 0.1163 0.2448>
            <at 0.15 0.0477>
            <at 0.2091 0.25>
            <at 0.25 0.25>
            <at 0.455 0.5>
            <at 0.6597 0.5>
            <at 0.7304 0.25>
            <at 0.8114 0.25>
            <at 0.8468 0.0741>
            <at 0.8983 0.0649>
            <at 0.9248 0.15>
            <at 1 0.15 RCIM_Cubic>
           ]}
          RedCurve {keys [
            {arriveTangent=1.3581, interpMode=RCIM_Cubic, time=0, value=0.15}
            <at 0.0736 0.25>
            <at 0.1163 0.2448>
            <at 0.15 0.1615>
            <at 0.2091 0.25>
            <at 0.25 0.25>
            <at 0.455 0.5>
            <at 0.6597 0.5>
            <at 0.7304 0.25>
            <at 0.8114 0.25>
            <at 0.8468 0.1094>
            <at 0.8983 0.1094>
            <at 0.9248 0.15>
            <at 1 0.15 RCIM_Cubic>
           ]}
         }}
        CurveIndex=Linked() {linkedVariable {name=NPC.FortniteNPC.TimeOfDay_0-1, type=NiagaraFloat}}
       ))
     )
     Scale Factor=<1 1 1 1>
    )) {isSetParameters=true, module=SetVariables_#3}
  ]
  rendererClasses [&NiagaraSpriteRendererProperties]
  renderers [{rendererClass=&NiagaraSpriteRendererProperties}]
 }
 Smoke002 {
  bEnabled=false
  emitterUpdateScript [
   &EmitterState(Life Cycle Mode=::NewEnumerator1, Loop Behavior=::NewEnumerator1, Loop Delay=1.25, Loop Duration=1.8, UseLoopDelay=true)
   &SpawnRate(SpawnRate=5, Use Spawn Probability=false)
  ]
  particleSpawnScript [
   &InitializeParticle(
    Color=<rgba 0.0192 0.0154 0.0365 1>
    Lifetime=3
    Mesh Scale Mode=::NewEnumerator3
    Sprite Rotation Mode=::NewEnumerator2
    Sprite Size Mode=::NewEnumerator3
    Uniform Sprite Size=500
    UsePositionOffset=false
    Write Material Random=true
   )
   &CameraOffset(Camera Offset Amount=200)
   &ShapeLocation(
    MaskSpawnGroup=false
    Non Uniform Scale=<3 3 1>
    Offset=&Multiply_VectorByFloat(Vector=$System.WindDirection, Float=0)
    Offset Coordinate Space=::World
    Offset Mode=::NewEnumerator0
    Override Seed=false
    Rotation Coordinate Space=::World
    Shape Origin=&SimulationPosition()
    Sphere Radius=50
    UseOwnerScale=false
   ) {module=ShapeLocation001}
   &AddVelocity(Rotation Coordinate Space=::World, Velocity=$System.WindDirection, Velocity Speed Scale=0)
  ]
  particleUpdateScript [
   &ParticleState()
   &CurlNoiseForce(Noise Strength=0, Pan Noise=false)
   &SolveForcesAndVelocity(Clamp Velocity=false, Limit Acceleration=false) {module=SolveForcesAndVelocity001}
   &ScaleColor(
    Scale Alpha=&FloatFromCurve(
     FloatCurve=DataInterface() {propertyValues {Curve {keys [
         {interpMode=RCIM_Cubic, tangentMode=RCTM_User, time=0, value=0}
         {interpMode=RCIM_Cubic, tangentMode=RCTM_User, time=0.2, value=1}
         <at 0.8 0.95>
         {tangentMode=RCTM_User, time=1, value=0}
        ]}}}
     Scale Curve=1.5
    )
    Scale RGB=&MakeVectorFromLinearColorRGB(Color=<rgba 0 0 0 1>)
    ScaleA=true
    ScaleRGB=false
   )
   &SubUVAnimation(
    Start Frame:NiagaraInt32=&RandomRangeInt(Minimum=12, Maximum=20)
    SubUV Animation Mode:ENiagara_SubUVLookupMode=::NewEnumerator1
   )
   &DynamicMaterialParameters(
    Index 0 Param 1=250
    Module.Param0WriteEnabled=true
    Module.Param1WriteEnabled=true
    Module.Param2WriteEnabled=true
    Module.Param3WriteEnabled=true
   )
  ]
  rendererClasses [&NiagaraSpriteRendererProperties]
  renderers [{rendererClass=&NiagaraSpriteRendererProperties}]
 }
 Smoke003 {
  emitterUpdateScript [
   &EmitterState(Life Cycle Mode=::NewEnumerator1, Loop Behavior=::NewEnumerator1, Loop Duration=0.25, UseLoopDelay=false)
   &SpawnRate(SpawnRate=300, Use Spawn Probability=false)
  ]
  particleSpawnScript [
   &InitializeParticle(
    Color=<rgba 0.0192 0.0154 0.0365 1>
    Lifetime=2
    Mesh Scale Mode=::NewEnumerator3
    Sprite Rotation Mode=::NewEnumerator2
    Sprite Size Mode=::NewEnumerator3
    Uniform Sprite Size=2000
    UsePositionOffset=false
    Write Material Random=true
   )
   &CameraOffset(Camera Offset Amount=75)
   &ShapeLocation(
    MaskSpawnGroup=false
    Non Uniform Scale=<3 3 1>
    Offset=&Multiply_VectorByFloat(Vector=$System.WindDirection, Float=10000)
    Offset Coordinate Space=::World
    Offset Mode=::NewEnumerator0
    Override Seed=false
    Rotation Coordinate Space=::World
    Shape Origin=&SimulationPosition()
    Sphere Radius=1000
    UseOwnerScale=false
   ) {module=ShapeLocation001}
   &AddVelocity(Rotation Coordinate Space=::World, Velocity=$System.WindDirection, Velocity Speed Scale=-8000)
  ]
  particleUpdateScript [
   &ParticleState()
   &CurlNoiseForce(Noise Strength=0, Pan Noise=false)
   &SolveForcesAndVelocity(Clamp Velocity=false, Limit Acceleration=false) {module=SolveForcesAndVelocity001}
   &ScaleColor(
    Scale Alpha=&FloatFromCurve(
     FloatCurve=DataInterface() {propertyValues {Curve {keys [
         {interpMode=RCIM_Cubic, tangentMode=RCTM_User, time=0, value=0}
         {interpMode=RCIM_Cubic, tangentMode=RCTM_User, time=0.5, value=1}
         <at 0.8 0.95>
         {tangentMode=RCTM_User, time=1, value=0}
        ]}}}
     Scale Curve=1.5
    )
    Scale RGB=&MakeVectorFromLinearColorRGB(Color=<rgba 0 0 0 1>)
    ScaleA=true
    ScaleRGB=false
   )
   &SubUVAnimation(
    Start Frame:NiagaraInt32=&RandomRangeInt(Minimum=12, Maximum=20)
    SubUV Animation Mode:ENiagara_SubUVLookupMode=::NewEnumerator1
   )
   &DynamicMaterialParameters(
    Index 0 Param 1=250
    Module.Param0WriteEnabled=true
    Module.Param1WriteEnabled=true
    Module.Param2WriteEnabled=true
    Module.Param3WriteEnabled=true
   )
  ]
  rendererClasses [&NiagaraSpriteRendererProperties]
  renderers [{rendererClass=&NiagaraSpriteRendererProperties}]
 }
 SpeedLines {
  emitterUpdateScript [
   &EmitterState(Life Cycle Mode=::NewEnumerator1, Loop Behavior=::NewEnumerator1, Loop Duration=1.7, UseLoopDelay=false)
   &SpawnRate(SpawnRate=500, Use Spawn Probability=false)
  ]
  particleSpawnScript [
   &InitializeParticle(
    Color=<rgba 1 0.6417 0.6228 0.15>
    Lifetime Min=2
    Lifetime Mode=::NewEnumerator1
    Sprite Size Max=<6 20>
    Sprite Size Min=<10 10>
    Sprite Size Mode=::NewEnumerator2
    UsePositionOffset=false
    Write Material Random=true
   ) {module=InitializeParticle001}
   &CameraOffset(Camera Offset Amount=250)
   &ShapeLocation(
    MaskSpawnGroup=false
    Non Uniform Scale=<3 3 1>
    Offset=&Multiply_VectorByFloat(Vector=$System.WindDirection, Float=10000)
    Offset Coordinate Space=::World
    Offset Mode=::NewEnumerator0
    Override Seed=false
    Rotation Coordinate Space=::World
    Shape Origin=&SimulationPosition()
    Sphere Radius=1500
    UseOwnerScale=false
   ) {module=ShapeLocation001}
   &AddVelocity(Rotation Coordinate Space=::World, Velocity=$System.WindDirection, Velocity Speed Scale=-4000)
  ]
  particleUpdateScript [
   &ParticleState()
   &CurlNoiseForce(Noise Strength=1000, Pan Noise=false)
   &SolveForcesAndVelocity(Clamp Velocity=false, Limit Acceleration=false) {module=SolveForcesAndVelocity001}
   &ScaleSpriteSizeBySpeed(Max Scale Factor=<1 10>, Min Scale Factor=<0.01 0.05>, Velocity Threshold=1)
   &ScaleColor(
    Scale Alpha=&FloatFromCurve(FloatCurve=DataInterface() {propertyValues {Curve {keys [
         {arriveTangent=12.4693, interpMode=RCIM_Cubic, leaveTangent=12.4693, tangentMode=RCTM_User, time=0, value=0}
         {interpMode=RCIM_Cubic, tangentMode=RCTM_Break, time=0.1, value=1}
         {interpMode=RCIM_Cubic, tangentMode=RCTM_Break, time=0.85, value=1}
         {arriveTangent=-2.221, interpMode=RCIM_Cubic, leaveTangent=-2.221, tangentMode=RCTM_Break, time=1, value=0}
        ]}}})
    Scale RGB=&VectorFromFloat(Value=1)
    ScaleA=true
    ScaleRGB=true
   ) {module=ScaleColor001}
  ]
  rendererClasses [&NiagaraSpriteRendererProperties]
  renderers [{rendererClass=&NiagaraSpriteRendererProperties}]
 }
 SpeedLines002 {
  bEnabled=false
  emitterUpdateScript [
   &EmitterState()
   &SpawnRate(SpawnRate=25, Use Spawn Probability=false)
  ]
  particleSpawnScript [
   &FN_SetSpawnGroup(Number of Spawn Groups=2)
   &InitializeParticle(
    Color=&MakeLinearColorFromVectorAndFloat("Vector (RGB)"=<1 1 1>, "Float (Alpha)"=&RandomRangeFloat(Minimum=0.3, Maximum=0.5))
    Lifetime Max=0.16
    Lifetime Min=0.1
    Lifetime Mode=::NewEnumerator1
    Sprite Size Max=<3 10>
    Sprite Size Min=<2 2>
    Sprite Size Mode=::NewEnumerator2
    UsePositionOffset=false
    Write Material Random=true
   )
   /SetVariables(Particles.Position=$Particles.ShapeLocation001.ShapePosition) {enabled=false, isSetParameters=true}
   &AddVelocity(Velocity=&Multiply_VectorByFloat(Vector=$System.WindDirection, Float=&RandomRangeFloat(Minimum=-100, Maximum=-250)))
   &ShapeLocation(MaskSpawnGroup=false, Override Seed=false, UseOwnerScale=true)
  ]
  particleUpdateScript [
   &SolveForcesAndVelocity(Clamp Velocity=false, Limit Acceleration=false)
   &ParticleState()
   &ScaleSpriteSize(Scale Factor=&Vector2DFromCurve(
     Vector2Curve=DataInterface() {propertyValues {XCurve {keys [<at 0 0.5 RCIM_Cubic>,<at 0.5 1 RCIM_Cubic>]}, YCurve {keys [<at 0 0 RCIM_Cubic>,<at 0.5 1 RCIM_Cubic>]}}}
     ScaleCurve=&Vector2DFromFloat(Value=1)
    ))
   &ScaleSpriteSizeBySpeed(Max Scale Factor=<1 10>, Min Scale Factor=<0.01 0.05>, Velocity Threshold=1)
   &ScaleColor(
    Scale Alpha=&Multiply_Float(A=&MakeFloatFromVector(Vector=$Particles.ShapePos, Channel=::NewEnumerator1), B=5)
    ScaleA=true
    ScaleRGB=true
   ) {enabled=false, module=ScaleColor001}
   &ScaleColor(
    Scale Alpha=&FloatFromCurve(
     FloatCurve=DataInterface() {propertyValues {Curve {keys [
         {arriveTangent=-3.5623, interpMode=RCIM_Cubic, leaveTangent=-3.5623, tangentMode=RCTM_User, time=0, value=1}
         {arriveTangent=-0.3548, tangentMode=RCTM_User, time=1, value=0}
        ]}}}
     CurveIndex=&MakeFloatFromVector(Vector=$Particles.ShapePos, Channel=::NewEnumerator2)
     Scale Curve=15
    )
    ScaleA=true
    ScaleRGB=true
   ) {module=ScaleColor002}
   &ScaleColor(
    Scale Alpha=&FloatFromCurve(FloatCurve=DataInterface() {propertyValues {Curve {keys [<at 0 0 RCIM_Cubic>, <at 0.4 1>, {arriveTangent=-2.221, interpMode=RCIM_Cubic, leaveTangent=-2.221, tangentMode=RCTM_User, time=1, value=0}]}}})
    Scale RGB=&VectorFromFloat(Value=1)
    ScaleA=true
    ScaleRGB=true
   )
  ]
  rendererClasses [&NiagaraSpriteRendererProperties]
  renderers [{rendererClass=&NiagaraSpriteRendererProperties}]
 }
}
summary {
 systemName=NS_LimeRock_LensWind
 emitters [
  {bEnabled=false, emitterName=Smoke, rendererClasses [&NiagaraSpriteRendererProperties]}
  {bEnabled=false, emitterName=Debris001, simTarget=GPUComputeSim, rendererClasses [&NiagaraSpriteRendererProperties]}
  {bEnabled=false, emitterName=Smoke001, simTarget=GPUComputeSim, rendererClasses [&NiagaraSpriteRendererProperties]}
  {emitterName=Smoke003, rendererClasses [&NiagaraSpriteRendererProperties]}
  {emitterName=Small_Debris, rendererClasses [&NiagaraSpriteRendererProperties]}
  {bEnabled=false, emitterName=SpeedLines002, rendererClasses [&NiagaraSpriteRendererProperties]}
  {emitterName=SpeedLines, rendererClasses [&NiagaraSpriteRendererProperties]}
  {bEnabled=false, emitterName=PostProcess, rendererClasses [&NiagaraComponentRendererProperties]}
  {bEnabled=false, emitterName=Smoke002, rendererClasses [&NiagaraSpriteRendererProperties]}
 ]
 userVariables [
  User.bIsInWater {type=NiagaraBool}
  User.Landscape {
   type=NiagaraDataInterfaceLandscape
   description=""
   defaultValue {
    struct=NiagaraExt_VariableValue_DataInterface
    value {
     dataInterface=&self:NiagaraDataInterfaceLandscape_0
     dataInterfaceClass=&NiagaraDataInterfaceLandscape
    }
   }
  }
  User.RampupTimeSmoke {type=NiagaraFloat, defaultValue=1, description=""}
 ]
}
